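# Niagara idle effect for the PolarBlind Vein quadruped cosmetic companion, attaching audio-reactive fire meshes to its skeleton.
# Left and right detail bursts use partition-indexed arrays for per-bone placement, while continuously spawned spirals follow shoulders and thighs and a body flame follows spine_02.
# Music amplitude scales alpha on all four emitters and triggers their Timeline modules above a threshold.
# A system flag selects time-of-day red tints or neutral colors.
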NiagaraSystem /CosmeticCompanions/FX/Builtins/PolarBlind/Idle/NS_Idle_PolarBlind_Vein
dependencies {
 usedDataInterfaces [
  &NiagaraDataInterfaceArrayFloat3
  &NiagaraDataInterfaceArrayInt32
  &NiagaraDataInterfaceCurve
  &NiagaraDataInterfaceArrayFloat
 ]
 usedDynamicInputs []
 usedModules [
  &SystemState
  &self:SystemSpawnScript.SystemScriptSource.SystemScriptGraph.NiagaraNodeAssignment_0.SetVariables_#1
  &self:SystemSpawnScript.SystemScriptSource.SystemScriptGraph.NiagaraNodeAssignment_2.SetVariables_#2
 ]
}
emitters {
 FireDetail_Left {
  emitterUpdateScript [
   &EmitterState()
   &SpawnBurst_Instantaneous(Spawn Count=10, Use Spawn Probability=false, UseLoopCountLimit=false)
  ]
  particleSpawnScript [
   &PartitionParticles(Number Of Partitions=10)
   &InitializeParticle(
    Color=&MakeCustomLinearColorFromBool(
     Bool=$System.Is Color Red
     True Color=&FNTimeOfDayColor(
      Night Color=<rgba 1.3 0.1383 0.2378 1>
      Morning Color=<rgba 1.5 0.1596 0.2744 1>
      Day Color=<rgba 2 0.2128 0.3658 1>
      Evening Color=<rgba 1.5 0.1596 0.2744 1>
     )
     False Color=<rgba 1 1 1 1>
    )
    Lifetime=5
    Mesh Scale=&Multiply_Vector(
     A=&SelectVectorFromArray(
      Direct Array Index=$Particles.PartitionParticles.Partition
      Vector Selection Array=DataInterface() {propertyValues {FloatData [<1.3 1 1.5>, <0 0 0>, <0 0 0>, <0 0 0>, <0 0 0>, <1 1 1>, <0 0 0>, <0 0 0>, <0 0 0>, <1 1 1>]}}
     )
     B=<1 1 1>
    )
    Mesh Scale Mode=::NewEnumerator4
    UsePositionOffset=false
    Write Material Random=true
   )
   &InitialMeshOrientation(
    Mesh Orientation Mode=::NewEnumerator7
    Rotation=&MergeHelper_VectorSwitch(
     0=<-0.0572 -1.0706 0.0082>
     1=<0.049 1.3406 0.0082>
     2=<0 0 0>
     3=<0 0 0>
     4=<-0.3349 0 0.3186>
     5=<0 0.4656 -0.3431>
     6=<0 0 0>
     7=<0 0 0>
     8=<0.147 0 0>
     9=<0.4983 0.1144 0>
     Index=$Particles.PartitionParticles.Partition
     Num=::NewEnumerator8
    )
    Use Rotation Vector=true
   )
   &CameraOffset(Camera Offset Amount=&MergeHelper_FloatSwitch(0=6, 1=2, 2=0, 3=0, 4=5, 5=5, 6=0, 7=0, 8=3, 9=3, Index=$Particles.PartitionParticles.Partition, Num=::NewEnumerator8))
   /SetVariables(Particles.Flame Color Dark:LinearColor=<rgba 0 0 0 1>) {isSetParameters=true, module=SetVariables_#3}
  ]
  particleUpdateScript [
   &ParticleState()
   &SkeletalMeshLocation(
    Bone / Socket Index=&SelectIntFromArray(
     Direct Array Index=$Particles.PartitionParticles.Partition
     Int Selection Array=DataInterface() {propertyValues {IntData [0,0,1,2,3,4,5,6,7,8]}}
    )
    Bone Sampling Mode=::NewEnumerator4
    MaskSpawnGroup=false
    Sampled Position Offset=&SelectVectorFromArray(
     Direct Array Index=$Particles.PartitionParticles.Partition
     Vector Selection Array=DataInterface() {propertyValues {FloatData [<-2.4995 4 3.85>, <0 0 0>, <0 0 6>, <0 0 0>, <1.296 0 0>, <0 0 0>, <0 0 0>, <4 0 0>, <0 -1.1426 1.6>, <0 1.7402 0>]}}
    )
    Skeletal Mesh=$System.Skeletal Mesh
   )
   &DynamicMaterialParameters(
    Index 0 Param 1=&MergeHelper_FloatSwitch(0=0.4, 1=0.3, 2=0, 3=0, 4=0.184, 5=0.184, 6=0, 7=0, 8=0.184, 9=0.184, Index=$Particles.PartitionParticles.Partition, Num=::NewEnumerator8)
    Index 0 Param 2=&MakeCustomFloatFromBool(Bool=true, True Float=1.43, False Float=0.75)
    Index 0 Param 4=&SelectFloatFromArray(
     Direct Array Index=$Particles.PartitionParticles.Partition
     Float Selection Array=DataInterface() {propertyValues {FloatData [0,0.7,0.25,0.5,0.3,0,0,0,0,0,0]}}
    )
    Index 2 Param 2=&MergeHelper_FloatSwitch(0=25, 1=25, 2=0, 3=0, 4=50, 5=50, 6=0, 7=0, 8=50, 9=50, Index=$Particles.PartitionParticles.Partition, Num=::NewEnumerator8)
    Index 2 Param 3=0.9261
    Index 2 Param 4=&Lerp_Float(
     A=&FN_FrontEndFloatModifier(Gameplay=7, Front End=7)
     B=&FN_FrontEndFloatModifier(Gameplay=3, Front End=2.3)
     Alpha=$User.Variant
    )
    Module.Index0Float0Write=true
    Module.Index0Float1Write=true
    Module.Index0Float2Write=true
    Module.Index0Float3Write=true
    Module.Index2Float0Write=true
    Module.Index2Float1Write=true
    Module.Index2Float2Write=true
    Module.Index2Float3Write=true
    Write Parameter Index 2=::NewEnumerator1
   )
   &Timeline(Play=&SetBoolByFloatComparison(A=$NPC.NPC_Audio_Reactive_Parameters.Amplitude_Music, B=0.675), Rewind Rate=0.6)
   &ScaleColor(
    Scale Alpha=&Multiply_Float(
     A=&MergeHelper_FloatSwitch(0=1, 1=0, 2=0, 3=0, 4=0, 5=1.5, 6=0, 7=0, 8=0, 9=1.5, Index=$Particles.PartitionParticles.Partition, Num=::NewEnumerator8)
     B=&Lerp_Float(A=0.2, B=1, Alpha=$NPC.NPC_Audio_Reactive_Parameters.Amplitude_Music)
    )
    ScaleA=true
    ScaleRGB=true
   )
   /SetVariables(Particles.Flame Color Dark:LinearColor=&MultiplyLinearColor(
     Linear Color=$Particles.Initial.Flame Color Dark
     Scale Factor=&MakeVector4FromVectorAndFloat(
      "Vector (RGB)"=<1 1 1>
      "Float (Alpha)"=&Multiply_Float(
       A=&MergeHelper_FloatSwitch(0=1, 1=1, 2=0, 3=0, 4=1.5, 5=1.5, 6=0, 7=0, 8=1.5, 9=1.5, Index=$Particles.PartitionParticles.Partition, Num=::NewEnumerator8)
       B=&Lerp_Float(A=0.2, B=1, Alpha=$NPC.NPC_Audio_Reactive_Parameters.Amplitude_Music)
      )
     )
    )) {isSetParameters=true}
  ]
  rendererClasses [&NiagaraMeshRendererProperties]
  renderers [
   {rendererClass=&NiagaraMeshRendererProperties, rendererIndex=0}
   {rendererClass=&NiagaraMeshRendererProperties, rendererIndex=1}
  ]
 }
 FireDetail_Right {
  emitterUpdateScript [
   &EmitterState()
   &SpawnBurst_Instantaneous(Spawn Count=10, Use Spawn Probability=false, UseLoopCountLimit=false)
  ]
  particleSpawnScript [
   &PartitionParticles(Number Of Partitions=10)
   &InitializeParticle(
    Color=&MakeCustomLinearColorFromBool(
     Bool=$System.Is Color Red
     True Color=&FNTimeOfDayColor(
      Night Color=<rgba 1.3 0.1383 0.2378 1>
      Morning Color=<rgba 1.5 0.1596 0.2744 1>
      Day Color=<rgba 2 0.2128 0.3658 1>
      Evening Color=<rgba 1.5 0.1596 0.2744 1>
     )
     False Color=<rgba 1 1 1 1>
    )
    Lifetime=5
    Mesh Scale=&Multiply_Vector(
     A=&SelectVectorFromArray(
      Direct Array Index=$Particles.PartitionParticles.Partition
      Vector Selection Array=DataInterface() {propertyValues {FloatData [<0 0 0>, <1.3 1 1.5>, <0 0 0>, <0 0 0>, <1 1 0.9213>, <0 0 0>, <0 0 0>, <0 0 0>, <1 0.3587 0.65>, <0 0.6 0.7>]}}
     )
     B=<1 1 1>
    )
    Mesh Scale Mode=::NewEnumerator4
    UsePositionOffset=false
    Write Material Random=true
   )
   &InitialMeshOrientation(
    Mesh Orientation Mode=::NewEnumerator7
    Rotation=&MergeHelper_VectorSwitch(
     0=<0 0 0>
     1=<0.049 0.049 0.0082>
     2=<0 0 0>
     3=<0 0 0>
     4=<-0.3349 0 0.3186>
     5=<0 0 0>
     6=<0 0 0>
     7=<0 0 0>
     8=<0.147 -0.0082 -0.0735>
     9=<0 0 0>
     Index=$Particles.PartitionParticles.Partition
     Num=::NewEnumerator8
    )
    Use Rotation Vector=true
   )
   &CameraOffset(Camera Offset Amount=&MergeHelper_FloatSwitch(0=0, 1=6, 2=0, 3=0, 4=5, 5=0, 6=0, 7=0, 8=3, 9=0, Index=$Particles.PartitionParticles.Partition, Num=::NewEnumerator8))
   /SetVariables(Particles.Flame Color Dark:LinearColor=<rgba 0 0 0 1>) {isSetParameters=true, module=SetVariables_#3}
  ]
  particleUpdateScript [
   &ParticleState()
   &SkeletalMeshLocation(
    Bone / Socket Index=&SelectIntFromArray(
     Direct Array Index=$Particles.PartitionParticles.Partition
     Int Selection Array=DataInterface() {propertyValues {IntData [0,0,1,2,3,4,5,6,7,8]}}
    )
    Bone Sampling Mode=::NewEnumerator4
    MaskSpawnGroup=false
    Sampled Position Offset=&SelectVectorFromArray(
     Direct Array Index=$Particles.PartitionParticles.Partition
     Vector Selection Array=DataInterface() {propertyValues {FloatData [<0 0 0>, <2.419 4 3.8526>, <0 0 6>, <0 0 0>, <1.296 0 0>, <0 0 0>, <0 0 0>, <4 0 0>, <0 -2.695 1.6>, <0 0 0>]}}
    )
    Skeletal Mesh=$System.Skeletal Mesh
   )
   &DynamicMaterialParameters(
    Index 0 Param 1=&MergeHelper_FloatSwitch(0=0, 1=0.4, 2=0, 3=0, 4=0.184, 5=0, 6=0, 7=0, 8=0.184, 9=0, Index=$Particles.PartitionParticles.Partition, Num=::NewEnumerator8)
    Index 0 Param 2=&MakeCustomFloatFromBool(Bool=true, True Float=1.43, False Float=0.75)
    Index 0 Param 4=&SelectFloatFromArray(
     Direct Array Index=$Particles.PartitionParticles.Partition
     Float Selection Array=DataInterface() {propertyValues {FloatData [0,0.7,0.25,0.5,0.3,0,0,0,0,0,0]}}
    )
    Index 2 Param 2=&MergeHelper_FloatSwitch(0=0, 1=25, 2=0, 3=0, 4=50, 5=0, 6=0, 7=0, 8=50, 9=0, Index=$Particles.PartitionParticles.Partition, Num=::NewEnumerator8)
    Index 2 Param 3=0.9261
    Index 2 Param 4=&Lerp_Float(
     A=&FN_FrontEndFloatModifier(Gameplay=7, Front End=7)
     B=&FN_FrontEndFloatModifier(Gameplay=3, Front End=2.3)
     Alpha=$User.Variant
    )
    Module.Index0Float0Write=true
    Module.Index0Float1Write=true
    Module.Index0Float2Write=true
    Module.Index0Float3Write=true
    Module.Index2Float0Write=true
    Module.Index2Float1Write=true
    Module.Index2Float2Write=true
    Module.Index2Float3Write=true
    Write Parameter Index 2=::NewEnumerator1
   )
   &Timeline(Play=&SetBoolByFloatComparison(A=$NPC.NPC_Audio_Reactive_Parameters.Amplitude_Music, B=0.675), Rewind Rate=0.6)
   &ScaleColor(
    Scale Alpha=&Multiply_Float(
     A=&MergeHelper_FloatSwitch(0=0, 1=1, 2=0, 3=0, 4=1.5, 5=0, 6=0, 7=0, 8=1.5, 9=0, Index=$Particles.PartitionParticles.Partition, Num=::NewEnumerator8)
     B=&Lerp_Float(A=0.2, B=1, Alpha=$NPC.NPC_Audio_Reactive_Parameters.Amplitude_Music)
    )
    ScaleA=true
    ScaleRGB=true
   )
   /SetVariables(Particles.Flame Color Dark:LinearColor=&MultiplyLinearColor(
     Linear Color=$Particles.Initial.Flame Color Dark
     Scale Factor=&MakeVector4FromVectorAndFloat(
      "Vector (RGB)"=<1 1 1>
      "Float (Alpha)"=&Multiply_Float(
       A=&MergeHelper_FloatSwitch(0=0, 1=1, 2=0, 3=0, 4=1.5, 5=0, 6=0, 7=0, 8=1.5, 9=0, Index=$Particles.PartitionParticles.Partition, Num=::NewEnumerator8)
       B=&Lerp_Float(A=0.2, B=1, Alpha=$NPC.NPC_Audio_Reactive_Parameters.Amplitude_Music)
      )
     )
    )) {isSetParameters=true}
  ]
  rendererClasses [&NiagaraMeshRendererProperties]
  renderers [
   {rendererClass=&NiagaraMeshRendererProperties, rendererIndex=0}
   {rendererClass=&NiagaraMeshRendererProperties, rendererIndex=1}
  ]
 }
 FireSpirals {
  emitterSpawnScript [/SetVariables(Emitter.TimeMult:NiagaraFloat=&RandomRangeFloat(Minimum=1.5, Maximum=2)) {isSetParameters=true}]
  emitterUpdateScript [
   &EmitterState(Inactive Response=::NewEnumerator1, Life Cycle Mode=::NewEnumerator1, Loop Duration=1, UseLoopDelay=false)
   &SpawnRate(SpawnRate=6, Use Spawn Probability=false)
  ]
  particleSpawnScript [
   &PartitionParticles(Number Of Partitions=4)
   &InitializeParticle(
    Color=&MakeCustomLinearColorFromBool(
     Bool=$System.Is Color Red
     True Color=&FNTimeOfDayColor(
      Night Color=<rgba 1 0.1064 0.1829 1>
      Morning Color=<rgba 1.2 0.1064 0.1829 1>
      Day Color=<rgba 1.8 0.1915 0.3292 1>
      Evening Color=<rgba 1.2 0.1064 0.1829 1>
     )
     False Color=&RandomRangeLinearColor(Minimum=<rgba 0.6354 0.6073 0.6023 1>, Maximum=<rgba 0.8021 0.7666 0.7603 1>)
    )
    Lifetime Max=1.65
    Lifetime Min=1.3
    Lifetime Mode=::NewEnumerator1
    Mesh Scale=&Multiply_Vector(
     A=&SelectVectorFromArray(
      Direct Array Index=$Particles.PartitionParticles.Partition
      Vector Selection Array=DataInterface() {propertyValues {FloatData [<1 1 1>, <1 1 1>, <1.3 1 1>, <1.3 1 1>]}}
     )
     B=&RandomRangeVector(Minimum=<0.4 0.5 0.5>, Maximum=<0.5 0.5 0.5>)
    )
    Mesh Scale Mode=::NewEnumerator4
    UsePositionOffset=false
    Write Material Random=true
   )
   &InitialMeshOrientation(
    Mesh Orientation Mode=::NewEnumerator2
    Rotation=&SelectVectorFromArray(
     Direct Array Index=$Particles.PartitionParticles.Partition
     Vector Selection Array=DataInterface() {propertyValues {FloatData [<0 0 0.5>, <0 0 0>, <0 0 0.5>, <0 0 0>]}}
    )
    Use Rotation Vector=true
   )
   &InitialMeshOrientation(
    Mesh Orientation Mode=::NewEnumerator2
    Rotation=&RandomRangeVector(Minimum=<0 0 0>, Maximum=<1 0 0>)
    Use Rotation Vector=true
   ) {module=InitialMeshOrientation001}
   &InitialMeshOrientation(
    Mesh Orientation Mode=::NewEnumerator2
    Rotation=&SelectVectorFromArray(
     Direct Array Index=$Particles.PartitionParticles.Partition
     Vector Selection Array=DataInterface() {propertyValues {FloatData [<0 0.01 0.85>, <0 0 0.85>, <0 0.01 0>, <0 -0.01 0>]}}
    )
    Use Rotation Vector=true
   ) {module=InitialMeshOrientation002}
   &CameraOffset(Camera Offset Amount=3)
  ]
  particleUpdateScript [
   &ParticleState()
   &ScaleMeshSize(Scale Factor=&MakeCustomVectorFromBool(Bool=$User.bIsFrontEndPreview, True Vector=<1 0.8 0.8>, False Vector=<1 1 1>))
   &SkeletalMeshLocation(
    Bone / Socket Index=&SelectIntFromArray(Direct Array Index=$Particles.PartitionParticles.Partition, Int Selection Array=DataInterface() {propertyValues {IntData [0,1,2,3]}})
    Bone Sampling Mode=::NewEnumerator4
    MaskSpawnGroup=false
    Sampled Position Offset=&SelectVectorFromArray(
     Direct Array Index=$Particles.PartitionParticles.Partition
     Vector Selection Array=DataInterface() {propertyValues {FloatData [<-2 0 0>, <2 0 0>, <8 0 1>, <-8 0 -1>]}}
    )
    Skeletal Mesh=DataInterface() {propertyValues {
      DefaultMesh=null
      ExcludeBoneName=root
      OptionalComponentClass=null
      PreviewMesh=/CosmeticCompanions/Assets/Quadruped/PolarBlind_Vein/Meshes/PolarBlind_Vein_Quadruped
      SkinningMode=SkinOnTheFly
      SoftSourceActor=null
      SourceMode=Default
      UvSetIndex=0
      WholeMeshLOD=-1
      bExcludeBone=false
      bReadDeformedGeometry=true
      bRequireCurrentFrameData=true
      ComponentTags []
      FilteredBones [shoulder_fr_l, shoulder_fr_r, thigh_bk_l, thigh_bk_r]
      FilteredSockets []
      MeshUserParameter {parameter {name=None, typeDefHandle {registeredTypeIndex=1}, varData []}}
      SamplingRegions []
     }}
   )
   /SetVariables(Particles.Width:NiagaraFloat=&RandomRangeFloat(Minimum=0.3, Maximum=0.5)) {isSetParameters=true}
   &DynamicMaterialParameters(
    Index 0 Param 1=&Multiply_Float(A=15, B=$Particles.Width)
    Index 0 Param 2=&Multiply_Float(A=15, B=$Particles.Width)
    Index 0 Param 3=&RandomRangeFloat(Minimum=0, Maximum=0.8)
    Index 1 Param 1=&RandomRangeFloat(Minimum=1, Maximum=2)
    Index 2 Param 1=&RandomRangeFloat(Minimum=0.7, Maximum=1.3)
    Module.Index0Param0Write=true
    Module.Index0Param1Write=true
    Module.Index0Param2Write=true
    Module.Index0Param3Write=true
    Module.Index1Param0Write=true
    Module.Index1Param1Write=true
    Module.Index1Param2Write=true
    Module.Index1Param3Write=true
    Module.Index2Param0Write=true
    Module.Index2Param1Write=true
    Module.Index2Param2Write=true
    Module.Index2Param3Write=true
    Write Parameter Index 1=true
    Write Parameter Index 2=true
   )
   &SolveForcesAndVelocity(Clamp Velocity=false, Limit Acceleration=false)
   &Timeline(Play=&SetBoolByFloatComparison(A=$NPC.NPC_Audio_Reactive_Parameters.Amplitude_Music, B=0.675), Rewind Rate=0.6)
   &ScaleColor(
    Scale Alpha=&Multiply_Float(
     A=&FloatFromCurve(
      FloatCurve=DataInterface() {propertyValues {Curve {keys [<at 0 0 RCIM_Cubic>,<at 0.4 1 RCIM_Cubic>,<at 1 0 RCIM_Cubic>]}}}
      Scale Curve=&RandomRangeFloat(Minimum=1, Maximum=1)
     )
     B=&Lerp_Float(A=0.2, B=2, Alpha=$NPC.NPC_Audio_Reactive_Parameters.Amplitude_Music)
    )
    Scale RGB=&MakeVectorFromLinearColorRGB(Color=&RandomRangeLinearColor(Minimum=<rgba 0.5521 0.5521 0.5521 1>, Maximum=<rgba 0.7083 0.7083 0.7083 1>))
    ScaleA=true
    ScaleRGB=true
   )
   &ScaleMeshSize(Scale Factor=&VectorFromFloat(Value=&MakeCustomFloatFromBool(
      Bool=$User.bisHandHidden
      True Float=&SelectFloatFromArray(
       Direct Array Index=$Particles.PartitionParticles.Partition
       Float Selection Array=DataInterface() {propertyValues {FloatData [1,1,1,1,0,1,0,1]}}
      )
      False Float=&SelectFloatFromArray(
       Direct Array Index=$Particles.PartitionParticles.Partition
       Float Selection Array=DataInterface() {propertyValues {FloatData [1,1,1,1,1,1,1,1]}}
      )
     ))) {module=ScaleMeshSize001}
  ]
  rendererClasses [&NiagaraMeshRendererProperties]
  renderers [{rendererClass=&NiagaraMeshRendererProperties, rendererIndex=0}]
 }
 Fire_Body {
  emitterSpawnScript [/SetVariables(
    Emitter.Base Color 1:LinearColor=<rgba 0.5938 0.5938 0.5938 1>
    Emitter.Base Color 2:LinearColor=<rgba 0 0 0 1>
    Emitter.Fire Center Color 1:LinearColor=<rgba 0 0 0 1>
    Emitter.Fire Center Color 2:LinearColor=<rgba 0.0677 0.0677 0.0677 1>
   ) {isSetParameters=true}]
  emitterUpdateScript [
   &EmitterState(
    Inactive Response=::NewEnumerator1
    Life Cycle Mode=::NewEnumerator1
    Loop Behavior=::NewEnumerator1
    Loop Duration Mode=::NewEnumerator1
    UseLoopDelay=false
   )
   &SpawnBurst_Instantaneous(Spawn Count=2, Use Spawn Probability=false, UseLoopCountLimit=false)
  ]
  particleSpawnScript [
   &PartitionParticles()
   &InitializeParticle(
    Color=&MakeCustomLinearColorFromBool(
     Bool=$System.Is Color Red
     True Color=&FNTimeOfDayColor(
      Night Color=<rgba 1 0.1064 0.1829 0.6>
      Morning Color=<rgba 1 0.1064 0.1829 0.7>
      Day Color=<rgba 1 0.1064 0.1829 0.8>
      Evening Color=<rgba 1 0.1064 0.1829 0.7>
     )
     False Color=<rgba 0.3177 0.3177 0.3177 0.6>
    )
    Mesh Scale=&SelectVectorFromArray(
     Direct Array Index=$Particles.PartitionParticles.Partition
     Vector Selection Array=DataInterface() {propertyValues {FloatData [<0 0 0>, <-0.5 -1 1>]}}
    )
    Mesh Scale Mode=::NewEnumerator4
    UsePositionOffset=false
    Write Material Random=true
   )
   &InitialMeshOrientation(
    Mesh Orientation Mode=::NewEnumerator2
    Rotation=&SelectVectorFromArray(
     Direct Array Index=$Particles.PartitionParticles.Partition
     Vector Selection Array=DataInterface() {propertyValues {FloatData [<0 0.25 0>, <0 0.25 0>]}}
    )
    Use Rotation Vector=true
   )
   &CameraOffset(Camera Offset Amount=2)
  ]
  particleUpdateScript [
   &ParticleState(Kill Particles When Lifetime Has Elapsed=false, Let Infinitely Lived Particles Die When Emitter Deactivates=true)
   &SkeletalMeshLocation(
    Bone / Socket Index=19
    Bone Sampling Mode=::NewEnumerator4
    MaskSpawnGroup=false
    Sampled Position Offset=<0 3 0>
    Skeletal Mesh=DataInterface() {propertyValues {
      DefaultMesh=null
      ExcludeBoneName=root
      OptionalComponentClass=null
      PreviewMesh=&PolarBlind_Vein_Quadruped
      SkinningMode=SkinOnTheFly
      SoftSourceActor=null
      SourceMode=Default
      UvSetIndex=0
      WholeMeshLOD=-1
      bExcludeBone=false
      bReadDeformedGeometry=true
      bRequireCurrentFrameData=true
      ComponentTags []
      FilteredBones [spine_02]
      FilteredSockets []
      MeshUserParameter {parameter {name=None, typeDefHandle {registeredTypeIndex=1}, varData []}}
      SamplingRegions []
     }}
   )
   &SolveForcesAndVelocity(Clamp Velocity=false, Limit Acceleration=false)
   &DynamicMaterialParameters(
    Index 0 Param 1=10
    Index 0 Param 2=8
    Index 0 Param 3=1
    Index 0 Param 4=1
    Module.Index0Float0Write=true
    Module.Index0Float1Write=true
    Module.Index0Float2Write=true
    Module.Index0Float3Write=true
   )
   &Timeline(Play=&SetBoolByFloatComparison(A=$NPC.NPC_Audio_Reactive_Parameters.Amplitude_Music, B=0.6), Rewind Rate=0.6)
   &ScaleColor(
    Scale Alpha=&Multiply_Float(
     A=&ClampFloat(Float=&FN_FrontEndFloatModifier(
       Gameplay=&ClampFloat(Float=&FN_LOD_Distance_Remap(Start Distance=200, End Distance=300, Minimum Output=1, Maximum Output=1))
       Front End=&FN_LOD_Distance_Remap(Start Distance=150, End Distance=200, Minimum Output=1, Maximum Output=1, Clamp Output=true)
      ))
     B=&Lerp_Float(A=0.2, B=1, Alpha=$NPC.NPC_Audio_Reactive_Parameters.Amplitude_Music)
    )
    ScaleA=true
    ScaleRGB=true
   )
  ]
  rendererClasses [&NiagaraMeshRendererProperties]
  renderers [{rendererClass=&NiagaraMeshRendererProperties, rendererIndex=0}]
 }
}
summary {
 systemName=NS_Idle_PolarBlind_Vein
 emitters [
  {emitterName=FireSpirals, rendererClasses [&NiagaraMeshRendererProperties]}
  {emitterName=Fire_Body, rendererClasses [&NiagaraMeshRendererProperties]}
  {emitterName=FireDetail_Right, rendererClasses [&NiagaraMeshRendererProperties]}
  {emitterName=FireDetail_Left, rendererClasses [&NiagaraMeshRendererProperties]}
 ]
 userVariables [
  User.bIsFrontEnd {type=NiagaraBool}
  User.bIsFrontEndPreview {type=NiagaraBool}
  User.bisHandHidden {type=NiagaraBool}
  User.bisHandsHidden {type=NiagaraBool}
  User.bisHeadHidden {type=NiagaraBool}
  User.BodyAccLOD {type=NiagaraInt32}
  User.BodyRef {
   type=NiagaraDataInterfaceSkeletalMesh
   defaultValue {
    struct=NiagaraExt_VariableValue_DataInterface
    value {dataInterface=&self:NiagaraDataInterfaceSkeletalMesh_1}
   }
  }
  User.FaceAccLOD {type=NiagaraInt32}
  User.HeadRef {
   type=NiagaraDataInterfaceSkeletalMesh
   defaultValue {
    struct=NiagaraExt_VariableValue_DataInterface
    value {dataInterface=&self:NiagaraDataInterfaceSkeletalMesh_0}
   }
  }
  User.Interpolate Intensity {type=NiagaraFloat, defaultValue=5, description=""}
  User.IsFrontEnd {type=NiagaraBool, defaultValue=true}
  User.IsLargeBody {type=NiagaraBool}
  User.Variant {type=NiagaraFloat, defaultValue=0, description=""}
 ]
}
